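# Editor utility widget example that imports a file through Interchange from a details view of its own properties and a button.
# The button builds a pipeline override list from a spline mesh pipeline path and a new PipeBP_Example with AutoCreateMaterial enabled.
# It imports ImportFile from a hard-coded Downloads folder, replacing existing assets, then syncs the Content Browser to the result.
# GetImportFile supplies four test FBX names.
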EditorUtilityWidgetBlueprint /Game/NeverCook/Interchange/Examples/EUW_ImportExample
graphs [
 EventGraph {nodes [
   "On Clicked (EditorUtilityButton_84)"() {is=title K2Node_ComponentBoundEvent}
   > RunImport()
   Event Construct()
   > SetObject(self=Get DetailsView_0() -> (DetailsView_0:Details View Object Reference), NewObject=Self-Reference())
  ]}
 GetImportFile {nodes [
   GetImportFile() {is=title K2Node_FunctionEntry}
   > Return Node!(
    execute
    NewParam:Array of Strings=Make Array!("[0]":String=TestImport40k.fbx, "[1]":String=TestImport15k.fbx, "[2]":String=TestImport5k.fbx, "[3]":String=TestImport.fbx) -> (Array:Array of Strings)
   )
  ]}
 RunImport {nodes [
   RunImport() {is=title K2Node_FunctionEntry}
   > n1 = Sequence()
   > Set l_Pipelines(l_Pipelines:Array of Soft Object Path Structures=Make Array!("[0]":Soft Object Path Structure=MakeSoftClassPath(PathString=&/Pipe_BR_FX_SplineMesh)) -> (Array:Array of Soft Object Path Structures)) -> (Output_Get:Array of Soft Object Path Structures)
   > n2 = Construct Object from Class(Class=&/PipeBP_Example.PipeBP_Example_C) -> (ReturnValue:Pipe BP Example Object Reference)
   > Set AutoCreateMaterial!(execute, AutoCreateMaterial:bool=true, self:Pipe BP Example Object Reference=@n2) -> (then, Output_Get:bool)
   > Add(
    TargetArray:Array of Soft Object Path Structures=Get l_Pipelines() -> (l_Pipelines:Array of Soft Object Path Structures)
    NewItem:"Soft Object Path Structure (by ref)"=GetSoftObjectPath(Object=@n2)
   ) {is=title K2Node_CallArrayFunction}
   n3 = GetInterchangeManagerScripted!(execute=@n1.then_1) -> (then, ReturnValue:Interchange Manager Object Reference)
   > n4 = CreateSourceData!(execute, InFileName:String=Append(A="C:\\Users\\bryce.lumpkin\\Downloads\\", B=Get ImportFile() -> (ImportFile:String)), None:%=@n3) -> (then, ReturnValue:Interchange Source Data Object Reference)
   > n5 = ImportAsset!(
    execute
    self:Interchange Manager Object Reference=@n3
    ContentPath:String=&/Imports
    SourceData:Interchange Source Data Object Reference=@n4
    ImportAssetParameters:"Import Asset Parameters Structure (by ref)"=Make Import Asset Parameters!(
     ReimportAsset:Object Reference
     ReimportSourceIndex:int=-1
     bIsAutomated:bool=true
     bFollowRedirectors:bool=true
     OverridePipelines:Array of Soft Object Path Structures=Get l_Pipelines()
     ImportLevel:Level Object Reference
     DestinationName:String
     bReplaceExisting:bool=True
     bForceShowDialog:bool=true
    ) -> (ImportAssetParameters:Import Asset Parameters Structure) {is=title K2Node_MakeStruct}
   ) -> (then, OutImportedObjects:Array of Object References, ReturnValue:bool)
   > Branch(Condition=Is Valid Index!(TargetArray:Array of Object References=@n5, IndexToTest:int) -> (ReturnValue:bool))
   n6 = CreateAssetData!(InAsset="Get (a copy)"(Array:Array of Object References=@n5) -> (Output:Object Reference), bAllowBlueprintClass=false) -> (
    ReturnValue_PackageName:Name
    ReturnValue_PackagePath:Name
    ReturnValue_AssetName:Name
    ReturnValue_AssetClassPath:Top Level Asset Path Structure
   )
   > SyncBrowserToObjects(AssetPaths=Make Array("[0]":String="To String (Name)"(InName=@n6)) -> (Array:Array of Strings))
  ]}
]
variables [ImportFile {type=string, default=""}]
widgets {
 info {rootWidgetClass=&StackBox, widgetCount=3}
 widgets [
  StackBox_0 {widgetClassPath=&StackBox, widget=StackBox_0, properties {orientation=Orient_Vertical}}
  DetailsView_0 {
   widgetClassPath=&DetailsView
   parent=StackBox_0
   widget=DetailsView_0
   properties {categoriesToShow [nocategory]}
   slot {class=StackBoxSlot}
  }
  EditorUtilityButton_84 {widgetClassPath=&EditorUtilityButton, parent=StackBox_0, widget=EditorUtilityButton_84, slot {class=StackBoxSlot}}
 ]
}
